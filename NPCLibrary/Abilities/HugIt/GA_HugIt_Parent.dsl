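; Parent NPC gameplay ability that, on a random chance, has an AI bot offer a hug emote and wait for a player to join in.
; It picks a random emote entry, pushes behavior control, plays the emote with an optional speech bubble and listens for emote interaction.
; Interaction ends it after the emote's duration; otherwise a timeout or failed roll can trigger other speech lines.
; EndAbility unbinds the event, stops the emote and pops behavior control.
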(event Ability|EventActivateAbility
  (bind _aicontroller (Variables|InternalData|GetAIController))
  (Ability|CommitAbility)
  (Utilities|FlowControl|Delay 0.2)
  (CallFunction|InitVariables)
  (if (<= (Math|Random|RandomFloat) (ScalableFloat|GetValueatLevel (Variables|Default|GetChanceToActivate)))
    (bind _success (CallFunction|SelectEmote))
    (if _success
      (AI|PushBehaviorControl _aicontroller -1048576)
      (Variables|InternalData|SetBehaviorPushed true)
      (Class|FortAthenaAIBotController|PlayEmote _aicontroller (Variables|InternalData|GetHugEmote))
      (Game|Player|BindEventtoOnEmoteInteracted (Variables|InternalData|GetAIPawn) (EventDispatchers|CreateEvent :!event Custom|EmoteInteracted))
      (if (Variables|Default|SpeechBubbles|GetPlaySpeechOnEmoteStart)
        (Class|BPFLAICommonFunctions|PlaySpeechBubble (Variables|InternalData|GetAIPawn) (Variables|InternalData|GetTag_OnEmoteStart) true (Variables|Default|GetTryPlayVOOnSpeechBubbleLines))
        (else
          _))
      (Utilities|FlowControl|Delay (ScalableFloat|GetValueatLevel (Variables|Default|GetDelayUntilStropTryingToHug)))
      (if (Variables|InternalData|GetbEmoteInitiated)
        _
        (elif (Variables|Default|SpeechBubbles|GetPlaySpeechOnNotEmoting)
          (Class|BPFLAICommonFunctions|PlaySpeechBubble (Variables|InternalData|GetAIPawn) (Variables|InternalData|GetTag_OnNotEmoting) true (Variables|Default|GetTryPlayVOOnSpeechBubbleLines))
          (Ability|EndAbility)
          (else
            _)))
      (else
        (Ability|EndAbility)))
    (elif (Variables|Default|SpeechBubbles|GetPlaySpeechOnConversionButNoEmoting)
      (Class|BPFLAICommonFunctions|PlaySpeechBubble (Variables|InternalData|GetAIPawn) (Variables|Default|SpeechBubbles|GetSpeechBubbleTag_OnConversionButNoEmote) true (Variables|Default|GetTryPlayVOOnSpeechBubbleLines))
      (else
        (Ability|EndAbility)))))

(event Ability|EventOnEndAbility
  (params (bWasCancelled :type Boolean))
  (bind _aipawn (Variables|InternalData|GetAIPawn))
  (bind _aicontroller (Variables|InternalData|GetAIController))
  ;[1 1 1] End Ability
  (Utilities|IsValid _aipawn
    (:Is%20Valid
      (Game|Player|UnbindEventfromOnEmoteInteracted _aipawn (EventDispatchers|CreateEvent :!event Custom|EmoteInteracted)))
    (:Is%20Not%20Valid
      _))
  (Utilities|IsValid _aicontroller
    (:Is%20Valid
      (Class|FortAthenaAIBotController|StopEmote _aicontroller)
      (if (Variables|InternalData|GetBehaviorPushed)
        (AI|PopBehaviorControl _aicontroller)))))

(event Custom|EmoteInteracted
  (params (Instigator :type Fort%20Player%20Pawn%20Object%20Reference))
  ;[1 1 1] AI started to Emote
  (Variables|InternalData|SetbEmoteInitiated true)
  (Utilities|FlowControl|Delay (ScalableFloat|GetValueatLevel (Variables|InternalData|GetDelayEmoting)))
  (Ability|EndAbility))

(fn InitVariables
  (Variables|InternalData|SetbEmoteInitiated false)
  (Variables|InternalData|SetBehaviorPushed false)
  (bind _outfortplayerpawn (CallFunction|GetAvatarasValidFortPlayerPawn)
    (:ValidFortPlayerPawn
      (bind _output_get (Variables|InternalData|SetAIPawn _outfortplayerpawn))
      (bind _asfort_athena_aibot_controller (Utilities|Casting|CastToFortAthenaAIBotController (Pawn|GetController _output_get))
        (:then
          (Variables|InternalData|SetAIController _asfort_athena_aibot_controller)
          (Variables|InternalData|SetSpeechBubbleComponent (Actor|GetComponentbyClass (Variables|InternalData|GetAIPawn) "/Script/FortniteGame.FortPawnComponent_SpeechBubble"))
          (return))
        (:CastFailed
          _)))
    (:AvatarCastFailed
      _))
  (Ability|EndAbility))

(fn SelectEmote
  (bind _emotedata (Variables|Default|GetEmoteData))
  (bind _returnvalue (Utilities|Array|Length _emotedata))
  (if (> _returnvalue 0)
    (bind _output_get (LocalVariables|SetSelectedEmote (Utilities|Array|Get(aref) (Variables|Default|GetEmoteData) (Math|Random|RandomIntegerinRange 0 (- _returnvalue 1)))))
    (bind (_output_get_emoteid_2_d151b8914974cebaddbdb38e5a035c40 _output_get_emoteduration_5_da784ff3403b58504bc19184acaa5cd0 _output_get_tagonrequestemote_14_67a8d55440865abddca7c49f017574f7 _output_get_tagondeniedemote_15_27e841c846d9aac7f9e0d1adad0ed99c) (Utilities|Struct|BreakSTRUCTHugItEmoteData _output_get))
    (if (AssetManager|IsValidPrimaryAssetId _output_get_emoteid_2_d151b8914974cebaddbdb38e5a035c40)
      (Variables|InternalData|SetHugEmote _output_get_emoteid_2_d151b8914974cebaddbdb38e5a035c40)
      (Variables|InternalData|SetDelayEmoting _output_get_emoteduration_5_da784ff3403b58504bc19184acaa5cd0)
      (Variables|InternalData|SetTag_OnEmoteStart _output_get_tagonrequestemote_14_67a8d55440865abddca7c49f017574f7)
      (Variables|InternalData|SetTag_OnNotEmoting _output_get_tagondeniedemote_15_27e841c846d9aac7f9e0d1adad0ed99c)
      (return true)
      (else
        (Fortnite|FortShippingLog self "HugIt Ability Trying to play Emote, but Emote Asset ID is Invalid." true)
        (return false)))
    (else
      (Fortnite|FortShippingLog self "HugIt Ability Trying to play Emote, but no Emotes were Specified in the Emote Data." true)
      (return false))))
